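# Physics asset for the F_MED_Raven female player body mesh in the Quill folder, covering only the pelvis and legs.
# Capsules wrap the thigh twist and calf bones, three per calf, while a zero-length capsule forms a sphere at the pelvis.
# All four leg bodies constrain directly to the pelvis with free swing and twist, so calves do not chain through thighs and the angular limits are inactive.
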PhysicsAsset /Game/Characters/Player/Female/Medium/Bodies/F_MED_Quill/Meshes/F_MED_Raven_Body_01_Physics
bodies {
 calf_l [
  {shapeType=Capsule, length=26.75, radius=5, center=<-17.3058 -2.7145 -1.3763>, rotation=<rot 88.1487 179.719 176.0444>}
  {shapeType=Capsule, length=5.5, radius=3.25, center=<-21.8718 -3.0208 -7.189>, rotation=<rot 88.1485 179.719 176.0444>}
  {shapeType=Capsule, length=10.25, radius=5, center=<-18.5136 -5.3994 -0.801>, rotation=<rot 86.222 176.6894 173.4695>}
 ]
 calf_r [
  {shapeType=Capsule, length=37.5, radius=5, center=<17.2761 0.701 -0.2788>, rotation=<rot 86.222 176.6894 173.4695>}
  {shapeType=Capsule, length=10.25, radius=5, center=<24.7403 1.155 4.5214>, rotation=<rot 86.222 176.6894 173.4695>}
  {shapeType=Capsule, length=10.25, radius=5, center=<16.584 4.6405 3.3113>, rotation=<rot 86.222 176.6894 173.4695>}
 ]
 pelvis [{shapeType=Capsule, length=0, radius=13.5, center=<-7.9989 -0.7404 0>}]
 thigh_twist_01_l [{shapeType=Capsule, length=40, radius=5.75, center=<7.3943 -0.9358 -0.4445>, rotation=<rot 87.1452 -178.9214 -177.8392>}]
 thigh_twist_01_r [{shapeType=Capsule, length=36.25, radius=6.25, center=<-9.4073 2.3424 -0.3829>, rotation=<rot 85.8197 176.3478 176.2116>}]
}
constraints [
 {
  bone1Name=thigh_twist_01_l
  bone2Name=pelvis
  swing1LimitDegrees=45
  swing1Motion=Free
  swing2LimitDegrees=45
  swing2Motion=Free
  twistLimitDegrees=45
  twistMotion=Free
 }
 {
  bone1Name=thigh_twist_01_r
  bone2Name=pelvis
  swing1LimitDegrees=45
  swing1Motion=Free
  swing2LimitDegrees=45
  swing2Motion=Free
  twistLimitDegrees=45
  twistMotion=Free
 }
 {
  bone1Name=calf_r
  bone2Name=pelvis
  swing1LimitDegrees=45
  swing1Motion=Free
  swing2LimitDegrees=45
  swing2Motion=Free
  twistLimitDegrees=45
  twistMotion=Free
 }
 {
  bone1Name=calf_l
  bone2Name=pelvis
  swing1LimitDegrees=45
  swing1Motion=Free
  swing2LimitDegrees=45
  swing2Motion=Free
  twistLimitDegrees=45
  twistMotion=Free
 }
]
